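# Post-process Anim Blueprint for the GazeboFun biped cosmetic companion, adding pose-asset shapes and secondary motion to its tufts.
# The input pose runs through a curve-driven PoseAsset blend, then four chained AnimDynamics nodes for the left and right two-bone tufts.
# Each node uses root-relative simulation with angular springs, tight limits and custom gravity.
# DynamicsAlpha is inverted on every node, so higher values reduce the simulation; the EventGraph is empty.
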AnimBlueprint /CosmeticCompanions/Assets/Biped/GazeboFun/Meshes/GazeboFun_Biped_Post_AnimBP
target_skeleton=&/GazeboFun_Biped_Skeleton
graphs [
 AnimGraph {nodes [
   n1 = Input Pose()
   n2 = Get PoseAsset() -> (PoseAsset:Pose Asset Object Reference)
   n3 = Get DynamicsAlpha() -> (DynamicsAlpha:float)
   n4 = Anim Dynamics - Bone: dyn_l_tuft_01(
    ComponentPose=Local To Component(LocalPose=Pose!(PoseAsset:Pose Asset Object Reference=@n2, SourcePose:Pose Link Structure=@n1) -> (Pose:Pose Link Structure) {is=title AnimGraphNode_PoseBlendNode, bIgnorePoseAssetSourceAnimationCurves=true, blendOption=Linear})
    Alpha=@n3
    ExternalForce="0,0,0"
   ) {
    angularBiasOverride=1
    angularDampingOverride=0.9999
    angularSpringConstant=2000
    bAngularSpring=true
    bOverrideAngularBias=true
    bOverrideAngularDamping=true
    bUseGravityOverride=true
    simulationSpace=RootRelative
    alphaBoolBlend {blendOption=Linear}
    alphaScaleBiasClamp {bias=1, scale=-1}
    boundBone {boneName=dyn_l_tuft_01}
    componentAppliedLinearAccClamp=<3000 3000 3000>
    gravityOverride {z=-240}
    physicsBodyDefinitions [{
      boundBone {boneName=dyn_l_tuft_01}
      boxExtents=<10 10 10>
      constraintSetup {angularLimitsMax=<0 15 15>, angularLimitsMin=<0 -15 -15>, angularTarget=<1 0 0>}
      localJointOffset=<2 0 0>
     }]
    simSpaceSettings {maxAngularAcceleration=40, maxAngularVelocity=8, simSpaceAngularAlpha=0.25}
   }
   n5 = Anim Dynamics - Bone: dyn_l_tuft_02(ComponentPose=@n4, Alpha=@n3, ExternalForce="0,0,0") {
    angularBiasOverride=1
    angularDampingOverride=0.9999
    angularSpringConstant=2000
    bAngularSpring=true
    bOverrideAngularBias=true
    bOverrideAngularDamping=true
    bUseGravityOverride=true
    simulationSpace=RootRelative
    alphaBoolBlend {blendOption=Linear}
    alphaScaleBiasClamp {bias=1, scale=-1}
    boundBone {boneName=dyn_l_tuft_02}
    componentAppliedLinearAccClamp=<3000 3000 3000>
    gravityOverride {z=-240}
    physicsBodyDefinitions [{
      boundBone {boneName=dyn_l_tuft_02}
      boxExtents=<10 10 10>
      constraintSetup {angularLimitsMax=<0 15 15>, angularLimitsMin=<0 -15 -15>, angularTarget=<1 0 0>}
      localJointOffset=<2 0 0>
     }]
    simSpaceSettings {maxAngularAcceleration=40, maxAngularVelocity=8, simSpaceAngularAlpha=0.25}
   }
   n6 = Anim Dynamics - Bone: dyn_r_tuft_01(ComponentPose=@n5, Alpha=@n3, ExternalForce="0,0,0") {
    angularBiasOverride=1
    angularDampingOverride=0.9999
    angularSpringConstant=2000
    bAngularSpring=true
    bOverrideAngularBias=true
    bOverrideAngularDamping=true
    bUseGravityOverride=true
    simulationSpace=RootRelative
    alphaBoolBlend {blendOption=Linear}
    alphaScaleBiasClamp {bias=1, scale=-1}
    boundBone {boneName=dyn_r_tuft_01}
    componentAppliedLinearAccClamp=<3000 3000 3000>
    gravityOverride {z=-240}
    physicsBodyDefinitions [{
      boundBone {boneName=dyn_r_tuft_01}
      boxExtents=<10 10 10>
      constraintSetup {angularLimitsMax=<0 15 15>, angularLimitsMin=<0 -15 -15>, angularTarget=<1 0 0>}
      localJointOffset=<-2 0 0>
     }]
    simSpaceSettings {maxAngularAcceleration=40, maxAngularVelocity=8, simSpaceAngularAlpha=0.25}
   }
   n7 = Anim Dynamics - Bone: dyn_r_tuft_02(ComponentPose=@n6, Alpha=@n3, ExternalForce="0,0,0") {
    angularBiasOverride=1
    angularDampingOverride=0.9999
    angularSpringConstant=2000
    bAngularSpring=true
    bOverrideAngularBias=true
    bOverrideAngularDamping=true
    bUseGravityOverride=true
    simulationSpace=RootRelative
    alphaBoolBlend {blendOption=Linear}
    alphaScaleBiasClamp {bias=1, scale=-1}
    boundBone {boneName=dyn_r_tuft_02}
    componentAppliedLinearAccClamp=<3000 3000 3000>
    gravityOverride {z=-240}
    physicsBodyDefinitions [{
      boundBone {boneName=dyn_r_tuft_02}
      boxExtents=<10 10 10>
      constraintSetup {angularLimitsMax=<0 15 15>, angularLimitsMin=<0 -15 -15>, angularTarget=<1 0 0>}
      localJointOffset=<-2 0 0>
     }]
    simSpaceSettings {maxAngularAcceleration=40, maxAngularVelocity=8, simSpaceAngularAlpha=0.25}
   }
   Output Pose(Result=Component To Local(ComponentPose=@n7))
  ]}
 EventGraph {links [], nodes []}
]
variables [PoseAsset {type=PoseAsset, default=&/GazeboFun_Biped_PoseAsset}]
